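# Translucent unlit Niagara sprite material for dot particles in the HighTower Date pickaxe FX, apparently a test version.
# Particle color supplies emissive, while opacity multiplies particle alpha, a filled-circle texture and a stripe mask projected on world XY, scaled by UVScale.
# The Erode dynamic parameter blends that mask toward white, so higher values fill in the circle.
# A world sphere-mask/checker branch and an emissive-adjustments function call feed nothing downstream.
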Material /BRCosmetics/Weapons/FORT_Melee/Pickaxe_HighTower_Date/FX/M_Hightower_TestDots
graph {
 properties {bUsedWithNiagaraSprites=true, blendMode=BLEND_Translucent, pixelDepthOffsetMode=PDOM_Legacy, shadingModel=MSM_Unlit, usedShadingModels=Unlit}
 nodes [
  n1 = VectorParameter() {group="4.2) Emissive Sphere Mask", parameterName=Location Offset, sortPriority=5, defaultValue=<rgba 0 0 0 1>}
  n2 = SphereMask(
   A=WorldPosition().XYZ
   B=Add(A=ActorPositionWS(), B=@n1.RGB)
   Radius=ScalarParameter() {defaultValue=20, group="4.2) Emissive Sphere Mask", parameterName=Sphere Mask Radius, sortPriority=3}
   Hardness=ScalarParameter() {group="4.2) Emissive Sphere Mask", parameterName=Sphere Mask Hardness, sortPriority=4}
  )
  n3 = ScalarParameter() {group="4.2) Emissive Sphere Mask", parameterName=Power, sortPriority=5}
  n4 = LinearInterpolate(
   A=ScalarParameter() {group="4.2) Emissive Sphere Mask", parameterName=Sphere Mask Low Value, sortPriority=1}
   B=ScalarParameter() {defaultValue=1, group="4.2) Emissive Sphere Mask", parameterName=Sphere Mask High Value, sortPriority=2}
   Alpha=Clamp(None=Power(Base=@n2, Exp=@n3) {constExponent=1})
  )
  n5 = ComponentMask(None=ActorPositionWS()) {g=true, r=true}
  n6 = ComponentMask(None=TextureSample(UVs=Multiply(A=@n5)) {texture=/Game/Environments/Props/Mansion/Textures/T_Mansion_Floor_Checker_2_D}.RGB) {g=true, r=true}
  LinearInterpolate(
   A=@n6
   B=Multiply(A=@n6, B=ComponentMask(None=@n4) {g=true, r=true})
   Alpha=ScalarParameter() {defaultValue=1, group="4.2) Emissive Sphere Mask", parameterName="Use Sphere Mask (World)", sortPriority=-1}
  )
  n7 = ParticleColor()
  MaterialFunctionCall(Emissive=@n7.RGB) {materialFunction=&MF_FX_EmissiveAdjustments}
  Reroute(None=@n7.RGB) {desc=Emissive}
  n8 = TextureSample() {texture=/Game/Athena/Prototype/E3Cutout/FilledCircle}
  n9 = Divide(A=WorldPosition().XYZ, B=ScalarParameter() {defaultValue=20, parameterName=UVScale})
  n10 = LinearInterpolate(
   A=TextureSample(UVs=ComponentMask(None=@n9) {g=true, r=true}) {samplerType=SAMPLERTYPE_Masks, texture=/Game/Weapons/WeaponSkins/Wraps/Textures/LizardStripes/T_Wraps_LizardStripes_H3}.RGB
   B=Constant3Vector() {constant=<rgba 1 1 1 0>}
   Alpha=DynamicParameter() {paramNames [Erode,Param2,Param3,Param4]}.Erode
  )
  Reroute(None=Multiply(A=@n7.A, B=Multiply(A=@n8.RGB, B=@n10))) {desc=Opacity}
 ]
}
